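# Default Lit, Nanite-enabled TV screen material that shows animated flipbooks: a fast, tiled flake pattern and a dance animation.
# A scalar Blend parameter lerps between the two time-driven flipbooks.
# The result feeds base color in the deprecated shared base shader function, whose diffuse and specular pass through ConvertFromDiffSpec.
# Destruction height-blend, coloration and concrete texture nodes are also present but do not feed that shader chain.
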Material /Game/Packages/Fortnite_UniqueMaterials_1/MH_Materials/M_Tv_Scree
graph {
 properties {bUsedWithNanite=true, opacityMaskClipValue=0.16, pixelDepthOffsetMode=PDOM_Legacy, usedShadingModels=Default Lit}
 nodes [
  n1 = TextureCoordinate() {coordinateIndex=1}
  n2 = TextureCoordinate() {uTiling=4, vTiling=4}
  n3 = TextureSample(UVs=@n2) {samplerType=SAMPLERTYPE_Normal, texture=/Game/Global/Textures/Default/Tileable_Patterns/BrokenTilingTextureConcrete_N}
  Add(
   A=Multiply(
    A=TextureSample(UVs=@n1) {samplerType=SAMPLERTYPE_Normal, texture=/Game/Global/Textures/Default/Masks_and_Gradients/RevealMask_n}.RGB
    B=Constant3Vector() {constant=<rgba 1 1 0 0>}
   )
   B=@n3.RGB
  )
  n4 = VectorParameter() {group=Diffuse, parameterName=DiffuseColoration, defaultValue=<rgba 1 1 1 0>}
  AppendVector(A=@n4.RGB, B=@n4.RGB)
  n5 = VectorParameter() {group=Code, parameterName=DiffuseColoration, defaultValue=<rgba 1 1 1 0>}
  AppendVector(A=@n5.RGB, B=@n5.RGB)
  n6 = TextureSample(UVs=TextureCoordinate() {uTiling=4, vTiling=4}) {samplerType=SAMPLERTYPE_LinearColor, texture=/Game/Global/Textures/Default/Noise/T_Noise01}
  n7 = TextureSample(UVs=TextureCoordinate() {coordinateIndex=1}) {samplerType=SAMPLERTYPE_LinearColor, texture=/Game/Global/Textures/Default/Masks_and_Gradients/RevealMask}
  n8 = Multiply(A=ConstantBiasScale(None=@n6.R) {bias=-0.1, scale=1}, B=@n7.R)
  n9 = ComponentMask(None=@n3.RGB) {b=true}
  n10 = Multiply(A=Multiply(A=@n7.R, B=OneMinus(None=@n9)), B=Constant() {r=2})
  Clamp(None=Add(A=Add(A=@n8, B=@n10), B=@n7.R))
  Constant() {r=1}
  LinearInterpolate()
  MaterialFunctionCall!(Texture, UVs, SubImages, Frame) {materialFunction=/Engine/Functions/Engine_MaterialFunctions03/Texturing/SubUV_Function}
  n11 = Frac(None=Multiply(A=Time(), B=Constant() {r=3}))
  n12 = Constant() {r=2}
  n13 = TextureObject() {texture=/Game/Global/Textures/Default/PointSplatter/BoatFlake}
  n14 = TextureCoordinate() {uTiling=3, vTiling=3}
  n15 = Frac(None=Time())
  n16 = Constant() {r=4}
  n17 = TextureObject() {texture=/Game/Effects/Fort_Effects/Textures/TV/T_TV_Dance}
  n18 = MaterialFunctionCall!(
   -----------Base Shader-----------
   BaseColor1=LinearInterpolate(
    A=MaterialFunctionCall("Animation  Phase (0-1)"=@n11, Number of Rows=@n12, Number of Columns=@n12, Texture=@n13, UVs=@n14) {materialFunction=&FlipBook}.Result
    B=MaterialFunctionCall("Animation  Phase (0-1)"=@n15, Number of Rows=@n16, Number of Columns=@n16, Texture=@n17) {materialFunction=&FlipBook}.Result
    Alpha=ScalarParameter() {parameterName=Blend}
   )
   Base Color Alpha=TextureSampleParameter2D() {group=Diffuse, parameterName=Diffuse, texture=&BoatFlake}.A
   SpecularMasks1
   Emissive
   Normals1
   World Position Offset
   DiffuseFresnelAmount
   -----------Destruction------------
   ------Multiple Diffuse Textures-----
   BaseColor2
   BaseColor3
   BaseColor4
   BaseColor5
   BaseColor6
   BaseColor7
   BaseColor8
   " - - - - - - - -"
   SpecularMasks2
   SpecularMasks3
   SpecularMasks4
   SpecularMasks5
   SpecularMasks6
   SpecularMasks7
   SpecularMasks8
   " - - - - - - - - "
   normal2
   Normal3
   Normal4
   Normal5
   Normal6
   Normal7
   Normal8
   " - - - - - - - - 1"
   Use 2 Textures
   Use 3 Textures
   Use 4 Textures
   Use 5 Textures
   Use 6 Textures
   Use 7 Textures
   Use 8 Textures
   Use 9 Textures
   Use 10 Textures
   ExposeLerpFunctionForInvertingTheHitDistanceCalc
  ) {materialFunction=/Game/Packages/Fortress_SharedMaterials/Base_Material_Functions/DEPRECATED_BaseShaderFunction}
  MaterialFunctionCall!(Power=@n18.Roughness) {materialFunction=/Engine/Functions/Engine_MaterialFunctions01/Shading/PowerToRoughness}
  MaterialFunctionCall(DiffuseColor=@n18.Diffuse, SpecularColor=Multiply(A=@n18.Specular) {constA=1, constB=0.04, desc=FresnelBaseReflectFraction}) {materialFunction=&ConvertFromDiffSpec}
  n19 = Multiply(A=@n6.R, B=Constant() {r=0.1})
  n20 = Clamp(None=ScalarParameter() {group=Code_Destruction, parameterName=DestructionAmount})
  n21 = Constant() {r=2}
  n22 = OneMinus(None=MaterialFunctionCall!(Lerp A, Lerp B, Height Texture=Add(A=@n7.R, B=@n19), Falloff Contrast=Constant() {r=2}, Transition Phase=Divide(A=@n20, B=@n21)) {materialFunction=/Game/Packages/Fortnight_MaterialFunctions/Temp/HeightLerp}.Lerp Alpha)
  Multiply(A=@n22, B=@n22)
  ScalarParameter() {group=Code, parameterName=Code_DestructionAmount}
  ScalarParameter() {group=Code, parameterName=WSDamageRadius}
  ScalarParameter() {group=Code, parameterName=Code_HitAnimation}
  ScalarParameter() {defaultValue=0.65, group=Diffuse, parameterName=DiffuseColorationThreshold}
  ScalarParameter() {group=WorldPositionOffset, parameterName=DistanceCullFadeAnimationOffset}
  StaticBoolParameter() {defaultValue=true, group=Diffuse, parameterName=UseProceduralDiffuseColorationMask}
  n23 = TextureSampleParameter2D() {group=Diffuse, parameterName=DiffuseColorationMask, texture=/Game/Global/Textures/Default/Masks_and_Gradients/16x}
  StaticComponentMaskParameter(None=AppendVector(A=@n23.RGB, B=@n23.RGB)) {defaultR=true, group=Diffuse, parameterName=DiffuseColorationMaskComponentMask}
  TextureSample(UVs=@n2) {texture=/Game/Global/Textures/Default/Tileable_Patterns/BrokenTilingTextureConcrete_D}
  TextureSample() {texture=&T_TV_Dance}
  VectorParameter() {parameterName=Code_DeformationVector}
  VectorParameter() {parameterName=WSDeformationCenter}
 ]
}
